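# Sound cue for the final demolition impact on small plastic objects, likely played when the object is destroyed.
# A mixer layers two random picks, one of three regular small plastic impacts at reduced volume and one of three final-impact waves at full volume.
# A modulator varies the pitch of the combined result slightly.
# It uses medium-object destruction attenuation and an impact concurrency set.
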SoundCue /Game/Sounds/Fort_Impact_Sounds/Demolition/Plastic/Demo_Plastic_Small_ImpactFinal_Cue
root=Modulator_13
nodes [
 {id=Mixer_4, children [Random_13,Random_14], properties {inputVolume [0.75,1]}}
 {id=Modulator_13, children [Mixer_4], properties {pitchMax=1.1, pitchMin=0.9}}
 {id=Random_13, children [WavePlayer_40, WavePlayer_41, WavePlayer_42], properties {weights [1,1,1]}}
 {id=Random_14, children [WavePlayer_43, WavePlayer_44, WavePlayer_45], properties {weights [1,1,1]}}
 {id=WavePlayer_40, properties {soundWaveAssetPtr=&/Plastic_Small_Impact01}}
 {id=WavePlayer_41, properties {soundWaveAssetPtr=&/Plastic_Small_Impact02}}
 {id=WavePlayer_42, properties {soundWaveAssetPtr=&/Plastic_Small_Impact03}}
 {id=WavePlayer_43, properties {soundWaveAssetPtr=&/Plastic_Small_ImpactFinal01}}
 {id=WavePlayer_44, properties {soundWaveAssetPtr=&/Plastic_Small_ImpactFinal02}}
 {id=WavePlayer_45, properties {soundWaveAssetPtr=&/Plastic_Small_ImpactFinal03}}
]
properties {
 attenuationSettings=/Game/Sounds/Attenuation/Fort_Demolition_Destroy_Medium_Sized_Object_Att
 maxDistance=4000
 pitchMultiplier=0.9
 soundClassObject=&Fort_Impacts_PSM
 volumeMultiplier=0.544
 concurrencyOverrides {resolutionRule=PreventNew}
 concurrencySet [&CONC_Object_Impacts_FourMax]
}
